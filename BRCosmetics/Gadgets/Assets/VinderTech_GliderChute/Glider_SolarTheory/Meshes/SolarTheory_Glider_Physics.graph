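# Physics asset for the Solar Theory glider cosmetic's skeletal mesh, giving it a simplified collision and joint setup.
# Four capsule bodies cover the Frame, Handles_Base and the mirrored left and right handle bones.
# Three constraints chain the handles to Handles_Base and that base to the Frame, all sharing the same swing and twist limit angles.
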PhysicsAsset /BRCosmetics/Gadgets/Assets/VinderTech_GliderChute/Glider_SolarTheory/Meshes/SolarTheory_Glider_Physics
bodies {
 Frame [{shapeType=Capsule, length=194.5918, radius=73.4585, center=<-0.0336 -19.9405 45.2616>, rotation=<rot 0 -0.1645 -85.2378>}]
 Handle_l [{shapeType=Capsule, length=41.3092, radius=27.3491, center=<-36.6434 -4.1278 -8.069>, rotation=<rot 0 -54.4031 -81.9777>}]
 Handle_r [{shapeType=Capsule, length=41.3092, radius=27.3491, center=<36.6381 4.1317 8.0698>, rotation=<rot 0 -54.4033 -81.9777>}]
 Handles_Base [{shapeType=Capsule, length=56.1551, radius=55.3444, center=<2.9645 17.1301 -6.6144>, rotation=<rot 0 104.2509 -160.8088>}]
}
constraints [
 {bone1Name=Handles_Base, bone2Name=Frame, swing1LimitDegrees=45, swing2LimitDegrees=45, twistLimitDegrees=45}
 {bone1Name=Handle_r, bone2Name=Handles_Base, swing1LimitDegrees=45, swing2LimitDegrees=45, twistLimitDegrees=45}
 {bone1Name=Handle_l, bone2Name=Handles_Base, swing1LimitDegrees=45, swing2LimitDegrees=45, twistLimitDegrees=45}
]
